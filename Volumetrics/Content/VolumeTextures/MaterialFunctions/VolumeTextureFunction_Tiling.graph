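# Material function that samples a tiling pseudo-volume texture, stored as a 2D atlas of slices, at a scaled position (world position by default).
# XY wraps inside each cell with an edge-padding remap, while Z picks two adjacent slices, assuming a square grid.
# Both samples use derivatives of the unwrapped XY to avoid wrap seams, then blend by Z fraction into RGB and alpha; the first slice's cell offset is also output.
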MaterialFunction /Volumetrics/Content/VolumeTextures/MaterialFunctions/VolumeTextureFunction_Tiling
graph {
 properties {bExposeToLibrary=true, description=Looks up a Volume texture from a given position input.}
 nodes [
  n1 = FunctionInput() {
   bUsePreviewValueAsDefault=true
   description=Acts as a divisor of the position input.
   inputName=Scale
   sortPriority=2
   previewValue=<100 100 100 1>
  }
  n2 = Divide(
   A=FunctionInput(Preview=WorldPosition().XYZ) {bUsePreviewValueAsDefault=true, description=The volume texture to use., inputName=Position, sortPriority=1}
   B=@n1
  )
  n3 = FunctionInput() {
   bUsePreviewValueAsDefault=true
   description=The target resolution
   inputName=Texture Resolution
   inputType=FunctionInput_Vector2
   sortPriority=4
   previewValue=<2048 0 0 1>
  }
  n4 = FunctionInput() {
   bUsePreviewValueAsDefault=true
   description=The number of frames in the volume texture.
   inputName=Num Frames
   inputType=FunctionInput_Vector2
   sortPriority=3
   previewValue=<12 12 0 1>
  }
  n5 = Divide(A=@n3, B=@n4) {constA=4096, constB=8, desc="2048 / 8 = 256  pixels per cell"}
  n6 = FunctionInput() {
   bUsePreviewValueAsDefault=true
   description=The number of pixels to use as edge padding to fix sub tiling within each frame.
   inputName=Edge Padding
   inputType=FunctionInput_Scalar
   sortPriority=5
   previewValue=<8 0 0 1>
  }
  n7 = Multiply(A=Frac(None=ComponentMask(None=@n2) {g=true, r=true}), B=Divide(A=@n5, B=Add(A=@n5, B=@n6) {constB=4, desc="Pad 4 pixels, 2 per side"}))
  n8 = Multiply(A=Divide(B=@n5) {desc=Pixel Size}, B=Multiply(A=@n6) {constB=0.5}) {constB=2}
  n9 = Multiply(A=@n4)
  n10 = Multiply(A=@n9)
  n11 = Divide(A=Frac(None=Add(A=@n7, B=@n8)), B=@n10) {constB=12}
  n12 = MaterialFunctionCall(Float2=@n4) {materialFunction=&BreakOutFloat2Components}
  n13 = Multiply(A=ComponentMask(None=@n2) {b=true}, B=Multiply(A=@n12.R, B=@n12.R))
  n14 = Floor(None=@n13)
  n15 = Divide(
   A=MaterialFunctionCall!(1D Index=@n14, Cells XY=@n9) {materialFunction=/Engine/Functions/Engine_MaterialFunctions01/Coordinates/1Dto2DIndex}.2D Index
   B=@n10
  ) {constB=12}
  n16 = FunctionInput(Preview=TextureObject() {texture=&Black}) {bUsePreviewValueAsDefault=true, description=The volume texture to use., inputName=Texture Object, inputType=FunctionInput_Texture2D}
  n17 = Divide(A=ComponentMask(None=@n2) {g=true, r=true}) {constB=8}
  n18 = DDX(Value=@n17)
  n19 = DDY(Value=@n17)
  n20 = TextureSample!(UVs=Add(A=@n11, B=@n15), Tex=@n16, "DDX(UVs)"=@n18, "DDY(UVs)"=@n19, Apply View MipBias) {constMipValue=0, mipValueMode=TMVM_Derivative, samplerType=SAMPLERTYPE_LinearColor}
  n21 = MaterialFunctionCall(1D Index=Add(A=@n14), Cells XY=@n9) {materialFunction=&1Dto2DIndex}
  n22 = TextureSample(UVs=Add(A=@n11, B=Divide(A=@n21.2D Index, B=@n10) {constB=12}), Tex=@n16, "DDX(UVs)"=@n18, "DDY(UVs)"=@n19) {constMipValue=0, mipValueMode=TMVM_Derivative, samplerType=SAMPLERTYPE_LinearColor}
  n23 = Frac(None=@n13) {desc=Z Blend Phase}
  FunctionOutput(None=LinearInterpolate(A=@n20.RGB, B=@n22.RGB, Alpha=@n23) {desc=Blend Z frames}) {outputName=Result RGB}
  FunctionOutput(None=@n15) {outputName=UV 1, sortPriority=1}
  FunctionOutput(None=LinearInterpolate(A=@n20.A, B=@n22.A, Alpha=@n23)) {outputName=Result Alpha, sortPriority=1}
 ]
}
