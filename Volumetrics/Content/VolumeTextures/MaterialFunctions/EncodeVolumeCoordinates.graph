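# EncodeVolumeCoordinates is a library material function that outputs 3D UVW coordinates laid out as a 2D grid of slices, for rendering to pseudo volume textures.
# It scales UVs by XY Frames, taking the fraction as in-slice UV and the floor as the 2D tile index.
# That index becomes a 1D slice index, offset to the slice center and normalized to W by X frames squared, which assumes a square grid.
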MaterialFunction /Volumetrics/Content/VolumeTextures/MaterialFunctions/EncodeVolumeCoordinates
graph {
 properties {bExposeToLibrary=true, description=Provides 3D UVW coordinates laid out in a 2D array for rendering to Psuedo Volume Textures.}
 nodes [
  n1 = FunctionInput(Preview=TextureCoordinate()) {bUsePreviewValueAsDefault=true, inputName=UVs, inputType=FunctionInput_Vector2, previewValue=<8 1 0 1>}
  n2 = FunctionInput() {bUsePreviewValueAsDefault=true, inputName=XY Frames, inputType=FunctionInput_Vector2, sortPriority=1, previewValue=<8 1 0 1>}
  n3 = Multiply(A=@n1, B=ComponentMask(None=@n2) {g=true, r=true})
  n4 = Floor(None=@n3)
  n5 = MaterialFunctionCall(Float2=@n2) {materialFunction=&BreakOutFloat2Components}
  n6 = Divide(
   A=Add(A=MaterialFunctionCall!(2D Index=@n4, Columns X=@n5.R) {materialFunction=/Engine/Functions/Engine_MaterialFunctions01/Coordinates/2Dto1DIndex}.1D Index) {constB=0.5}
   B=Multiply(A=@n5.R, B=@n5.R)
  )
  Add(A=@n6) {constB=0, desc=Precision Fix for converting back to 2d later. 0.00001}
  Ceil(None=@n3)
  FunctionOutput(None=AppendVector(A=Frac(None=@n3), B=@n6))
 ]
}
