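# Material function that sets up a volumetric raymarch by intersecting the camera ray with the primitive's box in 0-1 local space.
# A Custom HLSL node runs a slab test in object space, clamping the exit by scene depth, which is scaled by a hard-coded 1/256.
# The Steps and Plane alignment inputs snap the entry to step-spaced planes to reduce sampling artifacts.
# It outputs the entry position and the ray distance through the box.
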MaterialFunction /Volumetrics/Content/VolumeTextures/MaterialFunctions/VolumeBoxIntersect
graph {
 properties {bExposeToLibrary=true, description=Sets up a raymarch using a box intersection., libraryCategoriesText [Volumetrics]}
 nodes [
  MaterialExpressionFunctionInput_1 = FunctionInput() {
   bUsePreviewValueAsDefault=true
   description=Aligns start positions with plane intervals to reduce sampling artifacts.
   inputName=Plane alignment.
   inputType=FunctionInput_Scalar
   sortPriority=1
   previewValue=<1 0 0 1>
  }
  MaterialExpressionFunctionInput_0 = FunctionInput() {
   bUsePreviewValueAsDefault=true
   description=How many raymarch steps to use for Plane Alignment. Should match raymarch.
   inputName=Steps
   inputType=FunctionInput_Scalar
   previewValue=<128 0 0 1>
  }
  MaterialExpressionSceneDepth_1 = SceneDepth()
  n1 = Custom!(PlaneAlignment=@MaterialExpressionFunctionInput_1, MaxSteps=@MaterialExpressionFunctionInput_0, scenedepth=@MaterialExpressionSceneDepth_1.R) {
   code="float localscenedepth = scenedepth;\r\n\r\nfloat3 camerafwd = mul(float3(0.00000000,0.00000000,1.00000000),(float3x3)ResolvedView.ViewToTranslatedWorld);\r\n\r\nfloat3 depthpos = ((Parameters.CameraVector * localscenedepth) / abs( dot( camerafwd, Parameters.CameraVector ) ) );\r\n\r\ndepthpos = mul(depthpos,  (float3x3)WSDemote(GetWorldToLocal(Parameters))).xyz;\r\n\r\ndepthpos /= 256;\r\n\r\nlocalscenedepth = length(depthpos);\r\n\r\n//0-1\r\n//localscenedepth /= (GetPrimitiveData(Parameters).LocalObjectBoundsMax.x * 2 * scale);\r\n//localscenedepth /= abs( dot( camerafwd, Parameters.CameraVector ) );\r\n\r\n//bring vectors into local space to support object transforms\r\nfloat3 localcampos = mul(float4( DFDemote(ResolvedView.WorldCameraOrigin),1.00000000), (WSDemote(GetWorldToLocal(Parameters)))).xyz;\r\nfloat3 localcamvec = -normalize( mul(Parameters.CameraVector, (float3x3)WSDemote(GetWorldToLocal(Parameters))) );\r\n\r\n//make camera position 0-1\r\nlocalcampos = (localcampos / (GetPrimitiveData(Parameters).LocalObjectBoundsMax.x * 2)) + 0.5;\r\n\r\nfloat3 invraydir = 1 / localcamvec;\r\n\r\nfloat3 firstintersections = (0 - localcampos) * invraydir;\r\nfloat3 secondintersections = (1 - localcampos) * invraydir;\r\nfloat3 closest = min(firstintersections, secondintersections);\r\nfloat3 furthest = max(firstintersections, secondintersections);\r\n\r\nfloat t0 = max(closest.x, max(closest.y, closest.z));\r\nfloat t1 = min(furthest.x, min(furthest.y, furthest.z));\r\n\r\nfloat planeoffset = 1-frac( ( t0 - length(localcampos-0.5) ) * MaxSteps );\r\n\r\nt0 += (planeoffset / MaxSteps) * PlaneAlignment;\r\nt1 = min(t1, localscenedepth);\r\nt0 = max(0, t0);\r\n\r\nfloat boxthickness = max(0, t1 - t0);\r\n\r\nfloat3 entrypos = localcampos + (max(0,t0) * localcamvec);\r\n\r\n\r\nreturn float4( entrypos, boxthickness );"
   desc="\n*** Original source before expression upgrade ***\nfloat localscenedepth = scenedepth;\r\n\r\nfloat3 camerafwd = mul(float3(0.00000000,0.00000000,1.00000000),(float3x3)ResolvedView.ViewToTranslatedWorld);\r\n\r\nfloat3 depthpos = ((Parameters.CameraVector * localscenedepth) / abs( dot( camerafwd, Parameters.CameraVector ) ) );\r\n\r\ndepthpos = mul(depthpos,  (float3x3)GetPrimitiveData(Parameters).WorldToLocal).xyz;\r\n\r\ndepthpos /= 256;\r\n\r\nlocalscenedepth = length(depthpos);\r\n\r\n//0-1\r\n//localscenedepth /= (GetPrimitiveData(Parameters).LocalObjectBoundsMax.x * 2 * scale);\r\n//localscenedepth /= abs( dot( camerafwd, Parameters.CameraVector ) );\r\n\r\n//bring vectors into local space to support object transforms\r\nfloat3 localcampos = mul(float4( ResolvedView.WorldCameraOrigin,1.00000000), (GetPrimitiveData(Parameters).WorldToLocal)).xyz;\r\nfloat3 localcamvec = -normalize( mul(Parameters.CameraVector, (float3x3)GetPrimitiveData(Parameters).WorldToLocal) );\r\n\r\n//make camera position 0-1\r\nlocalcampos = (localcampos / (GetPrimitiveData(Parameters).LocalObjectBoundsMax.x * 2)) + 0.5;\r\n\r\nfloat3 invraydir = 1 / localcamvec;\r\n\r\nfloat3 firstintersections = (0 - localcampos) * invraydir;\r\nfloat3 secondintersections = (1 - localcampos) * invraydir;\r\nfloat3 closest = min(firstintersections, secondintersections);\r\nfloat3 furthest = max(firstintersections, secondintersections);\r\n\r\nfloat t0 = max(closest.x, max(closest.y, closest.z));\r\nfloat t1 = min(furthest.x, min(furthest.y, furthest.z));\r\n\r\nfloat planeoffset = 1-frac( ( t0 - length(localcampos-0.5) ) * MaxSteps );\r\n\r\nt0 += (planeoffset / MaxSteps) * PlaneAlignment;\r\nt1 = min(t1, localscenedepth);\r\nt0 = max(0, t0);\r\n\r\nfloat boxthickness = max(0, t1 - t0);\r\n\r\nfloat3 entrypos = localcampos + (max(0,t0) * localcamvec);\r\n\r\n\r\nreturn float4( entrypos, boxthickness );"
   description=Ray March Cube Setup
   outputType=CMOT_Float4
   inputs [
    {
     inputName=PlaneAlignment
     input {expression=&self:MaterialExpressionFunctionInput_1}
    }
    {
     inputName=MaxSteps
     input {expression=&self:MaterialExpressionFunctionInput_0}
    }
    {
     inputName=scenedepth
     input {expression=&self:MaterialExpressionSceneDepth_1, mask=1, maskR=1}
    }
   ]
  }
  FunctionOutput(None=ComponentMask(None=@n1) {a=true}) {outputName=Box Distance}
  FunctionOutput(None=ComponentMask(None=@n1) {b=true, g=true, r=true}) {outputName=Intersection Position, sortPriority=1}
 ]
}
