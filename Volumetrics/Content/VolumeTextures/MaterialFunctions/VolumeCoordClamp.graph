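# VolumeCoordClamp is a material function that pre-clamps UVs for a pseudo-volume texture laid out as a 2D grid of slices.
# A Custom HLSL node finds the XYFrames cell, redirecting the first and last Z slices to their inner neighbours.
# It clamps in-cell UVs by a 1.5-voxel border derived from VolumeResolution, then remaps to atlas space.
# A TexelSize input from Resolution is wired in but unused by the code.
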MaterialFunction /Volumetrics/Content/VolumeTextures/MaterialFunctions/VolumeCoordClamp
graph {nodes [
  MaterialExpressionTextureCoordinate_0 = TextureCoordinate()
  MaterialExpressionComponentMask_1 = ComponentMask(None=VectorParameter() {parameterName=XYFrames, defaultValue=<rgba 16 16 0 0>}.RGB) {g=true, r=true}
  MaterialExpressionDivide_5 = Divide(B=ComponentMask(None=VectorParameter() {parameterName=Resolution, defaultValue=<rgba 4096 4096 0 0>}.RGB) {g=true, r=true})
  MaterialExpressionDivide_6 = Divide(B=VectorParameter() {parameterName=VolumeResolution, defaultValue=<rgba 256 256 256 0>}.RGB) {constA=1.5, desc=Border Offset}
  MaterialExpressionOneMinus_3 = OneMinus(None=@MaterialExpressionDivide_6)
  n1 = Custom!(
   UV=@MaterialExpressionTextureCoordinate_0
   XYFrames=@MaterialExpressionComponentMask_1
   TexelSize=@MaterialExpressionDivide_5
   BorderMin=@MaterialExpressionDivide_6
   BorderMax=@MaterialExpressionOneMinus_3
  ) {
   code="//Psuedo Volume Coordinate Preclamp\r\nfloat2 Cellidx = floor(UV * XYFrames);\r\n\r\n//Clamp Z Frames\r\nCellidx = (Cellidx.x == 0 &&&& Cellidx.y == 0) ? float2(1,0) : Cellidx;\r\nCellidx = (Cellidx.x == (XYFrames.x - 1) &&&& Cellidx.y == (XYFrames.y - 1)) ? float2(XYFrames.x - 2, XYFrames.y - 1) : Cellidx;\r\n\r\nUV = frac(UV * XYFrames);\r\nUV = clamp(UV, BorderMin, BorderMax);\r\nUV /= XYFrames;\r\nUV += Cellidx / XYFrames;\r\n\r\nreturn UV;\r\n\r\n"
   description=Volume UV Pre Clamp
   outputType=CMOT_Float2
   inputs [
    {
     inputName=UV
     input {expression=&self:MaterialExpressionTextureCoordinate_0}
    }
    {
     inputName=XYFrames
     input {expression=&self:MaterialExpressionComponentMask_1}
    }
    {inputName=TexelSize, input {expression=&self:MaterialExpressionDivide_5}}
    {inputName=BorderMin, input {expression=&self:MaterialExpressionDivide_6}}
    {inputName=BorderMax, input {expression=&self:MaterialExpressionOneMinus_3}}
   ]
  }
  FunctionOutput(None=@n1) {outputName=UV}
 ]}
