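# VolumeEdgeMask is a library material function that fades a 3D 0-1 coordinate to black at the volume's edges, typically fed by EncodeVolumeCoordinates.
# It stretches Z by N²/(N²-1), using only the X frame count as N, to compensate for discrete slices.
# BoxMask-3D builds the falloff, then AlphaOffset applies capped hardness and Power shapes the curve.
# An older square-corner method remains behind a static switch not wired to the output.
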MaterialFunction /Volumetrics/Content/VolumeTextures/MaterialFunctions/VolumeEdgeMask
graph {
 properties {
  bExposeToLibrary=true
  description="This function creates a black gradient at the edges of 0 to 1 space in 3D. Can work with scaled worldposition, but typical input would be the function EncodeVolumeCoordinates."
 }
 nodes [
  n1 = FunctionInput() {
   bUsePreviewValueAsDefault=true
   description="The input 3D position to apply the mask to. Can be scaled worldposition, but typical input is the function EncodeVolumeCoordinates."
   inputName=UVW
  }
  n2 = Constant() {r=1}
  n3 = FunctionInput() {bUsePreviewValueAsDefault=true, inputName=XY Frames, inputType=FunctionInput_Vector2, sortPriority=10, previewValue=<8 8 0 1>}
  n4 = MaterialFunctionCall(Float2=@n3) {materialFunction=&BreakOutFloat2Components}
  n5 = Multiply(A=@n4.R, B=@n4.R)
  n6 = Subtract(A=@n5)
  n7 = Multiply(
   A=@n1
   B=MaterialFunctionCall(X=@n2, Y=@n2, Z=Divide(A=@n5, B=@n6)) {materialFunction=&MakeFloat3}.Result
  )
  n8 = FunctionInput() {
   bUsePreviewValueAsDefault=true
   description=The width of the edge gradient.
   inputName=Falloff Width
   sortPriority=1
   previewValue=<0.1 0.1 0.1 1>
  }
  n9 = Multiply(A=@n8) {constB=2}
  n10 = ComponentMask(None=MaterialFunctionCall!(A=@n7, B=Constant() {r=0.5}, Bounds=OneMinus(None=@n9), Edge Falloff=@n8) {materialFunction=/Engine/Functions/Engine_MaterialFunctions01/Gradient/BoxMask-3D}.Result) {r=true}
  n11 = FunctionInput() {
   bUsePreviewValueAsDefault=true
   description="How sharp the gradient is. A value of 1 would make the gradient appear black or white, starting at the edge width."
   inputName=Hardness
   inputType=FunctionInput_Scalar
   sortPriority=2
   previewValue=<0 1 1 1>
  }
  n12 = Min(A=@n11) {constB=0.999}
  n13 = FunctionInput() {
   bUsePreviewValueAsDefault=true
   description="The falloff exponent of the edge gradient. 1 = linear."
   inputName=Falloff Power
   inputType=FunctionInput_Scalar
   sortPriority=3
   previewValue=<1 1 1 1>
  }
  FunctionOutput(None=Power(
    Base=MaterialFunctionCall!(Alpha=@n10, Offset=@n12) {materialFunction=/Engine/Functions/Engine_MaterialFunctions01/Opacity/AlphaOffset}.Result
    Exp=@n13
   ))
  n14 = Divide(A=@n7, B=@n8)
  n15 = Divide(A=OneMinus(None=@n7), B=@n8)
  n16 = MaterialFunctionCall(Float3=Multiply(A=Clamp(None=@n14), B=Clamp(None=@n15))) {materialFunction=&BreakOutFloat3Components}
  StaticSwitch(True=Min(A=Min(A=@n16.R, B=@n16.R), B=@n16.B), False=@n10) {desc=Old Square Corners Method}
 ]
}
