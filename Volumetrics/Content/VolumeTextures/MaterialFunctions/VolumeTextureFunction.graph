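# Material function that samples a pseudo-volume texture stored as a 2D grid of Z slices at a given position.
# Position divided by Scale gives tiled XY within a frame, while Z picks two adjacent slices, counted as Num Frames X squared and optionally clamped.
# Both slices are sampled at a fixed mip and lerped by Z phase for filtered RGB and alpha, with UVs, frame and phase also output.
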MaterialFunction /Volumetrics/Content/VolumeTextures/MaterialFunctions/VolumeTextureFunction
graph {
 properties {bExposeToLibrary=true, description=Looks up a Volume texture from a given position input.}
 nodes [
  n1 = FunctionInput() {
   bUsePreviewValueAsDefault=true
   description=Acts as a divisor of the position input.
   inputName=Scale
   sortPriority=2
   previewValue=<100 100 100 1>
  }
  n2 = Divide(
   A=FunctionInput(Preview=WorldPosition().XYZ) {bUsePreviewValueAsDefault=true, description=The volume texture to use., inputName=Position, sortPriority=1}
   B=@n1
  )
  n3 = ComponentMask(None=@n2) {b=true}
  n4 = FunctionInput() {
   bUsePreviewValueAsDefault=true
   description=The number of frames in the volume texture.
   inputName=Num Frames
   inputType=FunctionInput_Vector2
   sortPriority=3
   previewValue=<12 12 0 1>
  }
  n5 = MaterialFunctionCall(Float2=@n4) {materialFunction=&BreakOutFloat2Components}
  n6 = Multiply(A=@n5.R, B=@n5.R)
  n7 = Add(A=Add(A=Multiply(A=@n3, B=@n6)) {constB=-0.5, desc=-0.5}, B=Multiply() {constA=1, constB=0}) {constA=1, constB=0}
  n8 = Floor(None=@n7)
  FunctionOutput(None=@n8) {outputName=z frame, sortPriority=50}
  n9 = Multiply(A=@n4)
  n10 = Multiply(A=@n9)
  n11 = Divide(A=Frac(None=ComponentMask(None=@n2) {g=true, r=true}), B=@n10) {constB=12}
  n12 = Constant()
  n13 = Subtract(A=@n6)
  n14 = FunctionInput(Preview=StaticBool()) {bUsePreviewValueAsDefault=true, inputName=Clamp Z, inputType=FunctionInput_StaticBool, sortPriority=100}
  n15 = StaticSwitch(True=Clamp(None=@n8, Min=@n12, Max=@n13), False=@n8, Value=@n14)
  n16 = Add(
   A=@n11
   B=Divide(
    A=MaterialFunctionCall!(1D Index=@n15, Cells XY=@n9) {materialFunction=/Engine/Functions/Engine_MaterialFunctions01/Coordinates/1Dto2DIndex}.2D Index
    B=@n10
   ) {constB=12}
  )
  n17 = FunctionInput(Preview=TextureObject() {texture=&Black}) {bUsePreviewValueAsDefault=true, description=The volume texture to use., inputName=Texture Object, inputType=FunctionInput_Texture2D}
  n18 = Constant() {r=8}
  n19 = TextureSample!(UVs=@n16, Tex=@n17, Level=@n18, Apply View MipBias) {constMipValue=0, mipValueMode=TMVM_MipLevel, samplerType=SAMPLERTYPE_LinearColor}
  n20 = Add(A=@n8)
  n21 = StaticSwitch(True=Clamp(None=@n20, Min=@n12, Max=@n13), False=@n20, Value=@n14)
  n22 = Add(
   A=@n11
   B=Divide(
    A=MaterialFunctionCall(1D Index=@n21, Cells XY=@n9) {materialFunction=&1Dto2DIndex}.2D Index
    B=@n10
   ) {constB=12}
  )
  n23 = TextureSample(UVs=@n22, Tex=@n17, Level=@n18) {constMipValue=0, mipValueMode=TMVM_MipLevel, samplerType=SAMPLERTYPE_LinearColor}
  n24 = Frac(None=@n7) {desc=Z Blend Phase}
  FunctionOutput(None=LinearInterpolate(A=@n19.RGB, B=@n23.RGB, Alpha=@n24) {desc=Blend Z frames}) {outputName=Result RGB}
  FunctionOutput(None=@n16) {outputName=UV 1, sortPriority=1}
  FunctionOutput(None=LinearInterpolate(A=@n19.A, B=@n23.A, Alpha=@n24)) {outputName=Result Alpha, sortPriority=1}
  FunctionOutput(None=@n22) {outputName=UV 2, sortPriority=5}
  FunctionOutput(None=@n24) {outputName=Z Phase, sortPriority=10}
 ]
}
