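# Material function that looks up a pseudo-volume texture, stored as a 2D atlas of slices, from a position input defaulting to world position.
# Position is divided by Scale; fractional XY gives the in-tile UV, while Z floors to one slice index, assuming a square Num Frames grid.
# 1Dto2DIndex converts that index to a tile offset, and one mip 0 sample is taken without blending between slices.
# Outputs include the RGB sample and atlas UV; DDX/DDY nodes are unconnected.
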MaterialFunction /Volumetrics/Content/VolumeTextures/MaterialFunctions/VolumeTextureFunction_NoZBlend
graph {
 properties {bExposeToLibrary=true, description=Looks up a Volume texture from a given position input.}
 nodes [
  n1 = FunctionInput() {
   bUsePreviewValueAsDefault=true
   description=Acts as a divisor of the position input.
   inputName=Scale
   sortPriority=2
   previewValue=<100 100 100 1>
  }
  n2 = Divide(
   A=FunctionInput(Preview=WorldPosition().XYZ) {bUsePreviewValueAsDefault=true, description=The volume texture to use., inputName=Position, sortPriority=1}
   B=@n1
  )
  n3 = FunctionInput() {
   bUsePreviewValueAsDefault=true
   description=The number of frames in the volume texture.
   inputName=Num Frames
   inputType=FunctionInput_Vector2
   sortPriority=3
   previewValue=<12 12 0 1>
  }
  n4 = Divide(A=ComponentMask(None=@n2) {g=true, r=true}, B=@n3) {constB=8}
  DDX(Value=@n4)
  DDY(Value=@n4)
  n5 = ComponentMask(None=@n2) {g=true, r=true}
  n6 = Multiply(A=@n3)
  n7 = Multiply(A=@n6)
  n8 = ComponentMask(None=@n2) {b=true}
  n9 = MaterialFunctionCall(Float2=@n3) {materialFunction=&BreakOutFloat2Components}
  n10 = Multiply(A=@n9.R, B=@n9.R)
  n11 = Floor(None=Subtract(A=Multiply(A=@n8, B=@n10)) {constB=0.5})
  n12 = Add(
   A=Divide(A=Frac(None=@n5), B=@n7) {constB=12}
   B=Divide(
    A=MaterialFunctionCall!(1D Index=@n11, Cells XY=@n6) {materialFunction=/Engine/Functions/Engine_MaterialFunctions01/Coordinates/1Dto2DIndex}.2D Index
    B=@n7
   ) {constB=12}
  )
  n13 = FunctionInput(Preview=TextureObject() {texture=&Black}) {bUsePreviewValueAsDefault=true, description=The volume texture to use., inputName=Texture Object, inputType=FunctionInput_Texture2D}
  n14 = TextureSample!(UVs=@n12, Tex=@n13, Level=Constant(), Apply View MipBias) {constMipValue=0, mipValueMode=TMVM_MipLevel, samplerType=SAMPLERTYPE_LinearColor}
  FunctionOutput(None=AppendVector(A=@n14.RGB, B=@n14.RGB)) {outputName=v4, sortPriority=1}
  FunctionOutput(None=@n14.RGB)
  FunctionOutput(None=@n12) {outputName=UV 1, sortPriority=1}
 ]
}
